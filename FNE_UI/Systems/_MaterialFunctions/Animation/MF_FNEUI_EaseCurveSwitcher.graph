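# Material function in the FNE UI animation library that applies a selectable easing curve to a 0-1 scalar.
# An index picks no ease or one of the In, Out and InOut variants of Sine, Cubic and Quint from MF_FNEUI_EaseCurves.
# Instead of branching, MF_FNEUI_CheckIntEquality results act as 0/1 alphas in a chain of lerps.
# Index 0 and unlisted integers fall through to the un-eased value.
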MaterialFunction /FNE_UI/Systems/_MaterialFunctions/Animation/MF_FNEUI_EaseCurveSwitcher
graph {
 properties {description=Select the Easing method you want to use, libraryCategoriesText [""]}
 nodes [
  n1 = NamedRerouteUsage()
  n2 = MaterialFunctionCall!(Int A=NamedRerouteUsage(), Int B=Constant() {desc="No Ease used (0)"}) {materialFunction=/FNE_UI/Systems/_MaterialFunctions/Operations/MF_FNEUI_CheckIntEquality}
  n3 = NamedRerouteUsage()
  n4 = Constant() {desc="EaseInSine (1)", r=1}
  n5 = LinearInterpolate(
   A=LinearInterpolate(A=@n1, B=@n1, Alpha=NamedRerouteDeclaration(None=@n2.Result) {name=is No Ease, desc="0 = No, 1 = Yes", nodeColor=<rgba 0 0 0 1>}) {desc="A = Any Integer not listed"}
   B=MaterialFunctionCall!("Value (0-1)"=NamedRerouteUsage()) {materialFunction=&/MF_FNEUI_EaseCurves}.EaseInSine
   Alpha=NamedRerouteDeclaration(None=MaterialFunctionCall(Int A=@n3, Int B=@n4) {materialFunction=&MF_FNEUI_CheckIntEquality}.Result) {name=is EaseInSine, desc="0 = No, 1 = Yes", nodeColor=<rgba 0 0 0 1>}
  )
  n6 = MaterialFunctionCall("Value (0-1)"=NamedRerouteUsage()) {materialFunction=&MF_FNEUI_EaseCurves}
  n7 = NamedRerouteDeclaration(None=MaterialFunctionCall(Int A=NamedRerouteUsage(), Int B=Constant() {desc="Ping Pong (2)", r=2}) {materialFunction=&MF_FNEUI_CheckIntEquality}.Result) {name=is EaseOutSine, desc="0 = No, 1 = Yes", nodeColor=<rgba 0 0 0 1>}
  n8 = NamedRerouteUsage()
  n9 = MaterialFunctionCall(Int A=NamedRerouteUsage(), Int B=Constant() {desc="EaseInOutSine (3)", r=3}) {materialFunction=&MF_FNEUI_CheckIntEquality}
  n10 = NamedRerouteUsage()
  n11 = Constant() {desc="EaseInCubic (4)", r=4}
  n12 = LinearInterpolate(
   A=LinearInterpolate(
    A=LinearInterpolate(A=@n5, B=@n6.EaseOutSine, Alpha=@n7)
    B=MaterialFunctionCall("Value (0-1)"=@n8) {materialFunction=&MF_FNEUI_EaseCurves}.EaseInOutSine
    Alpha=NamedRerouteDeclaration(None=@n9.Result) {name=is EaseInOutSine, desc="0 = No, 1 = Yes", nodeColor=<rgba 0 0 0 1>}
   )
   B=MaterialFunctionCall("Value (0-1)"=NamedRerouteUsage()) {materialFunction=&MF_FNEUI_EaseCurves}.EaseInCubic
   Alpha=NamedRerouteDeclaration(None=MaterialFunctionCall(Int A=@n10, Int B=@n11) {materialFunction=&MF_FNEUI_CheckIntEquality}.Result) {name=is EaseInCubic, desc="0 = No, 1 = Yes", nodeColor=<rgba 0 0 0 1>}
  )
  n13 = MaterialFunctionCall("Value (0-1)"=NamedRerouteUsage()) {materialFunction=&MF_FNEUI_EaseCurves}
  n14 = NamedRerouteDeclaration(None=MaterialFunctionCall(Int A=NamedRerouteUsage(), Int B=Constant() {desc="EaseOutCubic (5)", r=5}) {materialFunction=&MF_FNEUI_CheckIntEquality}.Result) {name=is EaseOutCubic, desc="0 = No, 1 = Yes", nodeColor=<rgba 0 0 0 1>}
  n15 = NamedRerouteUsage()
  n16 = MaterialFunctionCall(Int A=NamedRerouteUsage(), Int B=Constant() {desc="EaseInOutCubic (6)", r=6}) {materialFunction=&MF_FNEUI_CheckIntEquality}
  n17 = NamedRerouteUsage()
  n18 = Constant() {desc="EaseInQuint (7)", r=7}
  n19 = LinearInterpolate(
   A=LinearInterpolate(
    A=LinearInterpolate(A=@n12, B=@n13.EaseOutCubic, Alpha=@n14)
    B=MaterialFunctionCall("Value (0-1)"=@n15) {materialFunction=&MF_FNEUI_EaseCurves}.EaseInOutCubic
    Alpha=NamedRerouteDeclaration(None=@n16.Result) {name=is EaseInOutCubic, desc="0 = No, 1 = Yes", nodeColor=<rgba 0 0 0 1>}
   )
   B=MaterialFunctionCall("Value (0-1)"=NamedRerouteUsage()) {materialFunction=&MF_FNEUI_EaseCurves}.EaseInQuint
   Alpha=NamedRerouteDeclaration(None=MaterialFunctionCall(Int A=@n17, Int B=@n18) {materialFunction=&MF_FNEUI_CheckIntEquality}.Result) {name=is EaseInQuint, desc="0 = No, 1 = Yes", nodeColor=<rgba 0 0 0 1>}
  )
  n20 = MaterialFunctionCall("Value (0-1)"=NamedRerouteUsage()) {materialFunction=&MF_FNEUI_EaseCurves}
  n21 = NamedRerouteDeclaration(None=MaterialFunctionCall(Int A=NamedRerouteUsage(), Int B=Constant() {desc="EaseOutQuint (8)", r=8}) {materialFunction=&MF_FNEUI_CheckIntEquality}.Result) {name=is EaseOutQuint, desc="0 = No, 1 = Yes", nodeColor=<rgba 0 0 0 1>}
  n22 = NamedRerouteUsage()
  n23 = MaterialFunctionCall(Int A=NamedRerouteUsage(), Int B=Constant() {desc="EaseInOutQuint (9)", r=9}) {materialFunction=&MF_FNEUI_CheckIntEquality}
  FunctionOutput(None=LinearInterpolate(
    A=LinearInterpolate(A=@n19, B=@n20.EaseOutQuint, Alpha=@n21)
    B=MaterialFunctionCall("Value (0-1)"=@n22) {materialFunction=&MF_FNEUI_EaseCurves}.EaseInOutQuint
    Alpha=NamedRerouteDeclaration(None=@n23.Result) {name=is EaseInOutQuint, desc="0 = No, 1 = Yes", nodeColor=<rgba 0 0 0 1>}
   )) {outputName=Selected Easing Curve}
  n24 = FunctionInput(Preview=Constant()) {
   bUsePreviewValueAsDefault=true
   description=The value between 0 and 1 you want to ease
   inputName="Value to Ease (0-1)"
   inputType=FunctionInput_Scalar
  }
  NamedRerouteDeclaration(None=@n24) {name=Value to Ease, nodeColor=<rgba 0 0 0 1>}
  n25 = Constant() {desc="Easing Curve Switch: 0 = No Ease used, 1 =EaseInSine, 2 = EaseOutSine, 3 = EaseInOutSine, 4 = EaseInCubic, 5 = EaseOutCubic, 6 = EaseInOutCubic, 7 = EaseInQuint, 8 = EaseOutQuint, 9 = EaseInOutQuint"}
  n26 = FunctionInput(Preview=@n25) {
   bUsePreviewValueAsDefault=true
   description="Easing Curve Switch: 0 = No Ease used, 1 =EaseInSine, 2 = EaseOutSine, 3 = EaseInOutSine, 4 = EaseInCubic, 5 = EaseOutCubic, 6 = EaseInOutCubic, 7 = EaseInQuint, 8 = EaseOutQuint, 9 = EaseInOutQuint"
   inputName=Easing Curve Switch Index
   inputType=FunctionInput_Scalar
   sortPriority=1
  }
  NamedRerouteDeclaration(None=@n26) {name=Easing Curve Switch, nodeColor=<rgba 1 0 0.4833 1>}
 ]
}
