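# Material function in FNE_UI that outputs a UV set in a selectable scaling mode: none, Scaled to Fit or Scaled to Fill.
# MF_FNEUI_UVScaling computes the fit and fill variants from UV and texture size inputs.
# A scalar index is compared to constants through MF_FNEUI_CheckIntEquality, and the 0/1 results drive chained Lerps rather than static switches.
# Unlisted indices fall back to the first Lerp's unscaled value.
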MaterialFunction /FNE_UI/Systems/_MaterialFunctions/Basics/MF_FNEUI_UVScalingSwitcher
graph {
 properties {description=Select the UV Scaling Type you want to use, libraryCategoriesText [""]}
 nodes [
  n1 = NamedRerouteUsage()
  n2 = NamedRerouteUsage()
  n3 = NamedRerouteUsage()
  n4 = LinearInterpolate(
   A=@n1
   B=@n1
   Alpha=NamedRerouteDeclaration(None=MaterialFunctionCall!(Int A=@n2, Int B=@n3) {materialFunction=/FNE_UI/Systems/_MaterialFunctions/Operations/MF_FNEUI_CheckIntEquality}.Result) {name=is No UV Scaling, desc="0 = No, 1 = Yes", nodeColor=<rgba 0 0 0 1>}
  ) {desc="A = Any Integer not listed"}
  n5 = MaterialFunctionCall!(UV=NamedRerouteUsage(), TextureSize=NamedRerouteUsage(), ScalePivotPoint) {materialFunction=&/MF_FNEUI_UVScaling}
  n6 = NamedRerouteDeclaration(None=MaterialFunctionCall(Int A=NamedRerouteUsage(), Int B=NamedRerouteUsage()) {materialFunction=&MF_FNEUI_CheckIntEquality}.Result) {name=is Scaled to Fit, desc="0 = No, 1 = Yes", nodeColor=<rgba 0 0 0 1>}
  n7 = NamedRerouteUsage()
  n8 = NamedRerouteUsage()
  n9 = MaterialFunctionCall(Int A=NamedRerouteUsage(), Int B=NamedRerouteUsage()) {materialFunction=&MF_FNEUI_CheckIntEquality}
  FunctionOutput(None=LinearInterpolate(
    A=LinearInterpolate(A=@n4, B=@n5.Scaled to Fit, Alpha=@n6)
    B=MaterialFunctionCall(UV=@n7, TextureSize=@n8) {materialFunction=&MF_FNEUI_UVScaling}.Scaled To Fill
    Alpha=NamedRerouteDeclaration(None=@n9.Result) {name=is Scaled to Fill, desc="0 = No, 1 = Yes", nodeColor=<rgba 0 0 0 1>}
   )) {outputName=Selected Scaled UV}
  NamedRerouteDeclaration(None=FunctionInput(Preview=TextureCoordinate()) {bUsePreviewValueAsDefault=true, inputName=UV, inputType=FunctionInput_Vector2}) {name=UV, nodeColor=<rgba 0.9983 0.0094 0.0052 1>}
  n10 = FunctionInput() {
   bUsePreviewValueAsDefault=true
   description=Input the texture size as a vector2
   inputName=Texture Size
   inputType=FunctionInput_Vector2
   previewValue=<1 1 0 0>
  }
  NamedRerouteDeclaration(None=@n10) {name=Texture Size, nodeColor=<rgba 0.9983 0.0094 0.0052 1>}
  n11 = FunctionInput() {
   bUsePreviewValueAsDefault=true
   desc="UV Scaling Switch: 1 = No UV Scaling, 2 =Scaled to Fit, 3 = Scaled to Fill"
   description="1 = No UV Scaling, 2 =Scaled to Fit, 3 = Scaled to Fill"
   inputName=UV Scaling Switch Index
   inputType=FunctionInput_Scalar
   previewValue=<1 0 0 0>
  }
  NamedRerouteDeclaration(None=@n11) {name=UV Scaling Switch, nodeColor=<rgba 1 0 0.4833 1>}
  NamedRerouteDeclaration(None=Constant() {desc="No UV Scaling (1)", r=1}) {name=No UV Scaling Index, nodeColor=<rgba 0 0 0 1>}
  NamedRerouteDeclaration(None=Constant() {desc="Scaled to Fill (3)", r=3}) {name=Scaled to Fill Index, nodeColor=<rgba 0 0 0 1>}
  NamedRerouteDeclaration(None=Constant() {desc="Scaled to Fit (2)", r=2}) {name=Scaled to Fit Index, nodeColor=<rgba 0 0 0 1>}
 ]
}
