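# Translucent UI-domain material that draws a configurable rectangle with per-corner rounding, fill, stroke and an optional gap between them.
# Shape UVs support 9-slice or normalized input, scaling methods, slant and rotation.
# MF_FNEUI_Box calls produce fill, stroke and gap masks, which combine with colors and transparency into final color and alpha.
# Linear or radial gradients use separately scaled UVs, with optional time-driven, eased rotation.
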Material /FNE_UI/Systems/_Materials/Shapes/M_UI_Rectangle
graph {
 properties {blendMode=BLEND_Translucent, materialDomain=MD_UI, pixelDepthOffsetMode=PDOM_Legacy, usedShadingModels=Default Lit}
 nodes [
  n1 = ScalarParameter() {
   defaultValue=1
   desc="Default Value = 1.0\r\n\r\nSet the width of your shape from 0 to 1 (100%)\r\n"
   group="[1] Fill"
   parameterName=Width
   sliderMax=1
   sortPriority=5
  }
  n2 = NamedRerouteUsage()
  n3 = ScalarParameter() {
   defaultValue=1
   desc="Default Value = 1.0\r\n\r\nSet the height of your shape from 0 to 1 (100%) \r\n"
   group="[1] Fill"
   parameterName=Height
   sliderMax=1
   sortPriority=6
  }
  NamedRerouteDeclaration(None=AppendVector(A=Multiply(A=@n1, B=@n2), B=Multiply(A=@n2, B=@n3))) {name=Size, nodeColor=<rgba 0 0 0 1>}
  n4 = ScalarParameter() {
   desc="Default Value = 0.0\r\n\r\nSet the rounded corners of the shape.\r\n"
   group="[1] Fill"
   parameterName=CornerRoundnessTopLeft
   sliderMax=0.5
   sortPriority=10
  }
  n5 = ScalarParameter() {
   desc="Default Value = 0.0\r\n\r\nSet the rounded corners of the shape.\r\n"
   group="[1] Fill"
   parameterName=CornerRoundnessBottomLeft
   sliderMax=0.5
   sortPriority=8
  }
  n6 = ScalarParameter() {
   desc="Default Value = 0.0\r\n\r\nSet the rounded corners of the shape.\r\n"
   group="[1] Fill"
   parameterName=CornerRoundnessTopRight
   sliderMax=0.5
   sortPriority=11
  }
  n7 = ScalarParameter() {
   desc="Default Value = 0.0\r\n\r\nSet the rounded corners of the shape.\r\n"
   group="[1] Fill"
   parameterName=CornerRoundnessBottomRight
   sliderMax=0.5
   sortPriority=9
  }
  NamedRerouteDeclaration(None=MaterialFunctionCall(X=@n4, Y=@n5, Z=@n6, A=@n7) {materialFunction=&MakeFloat4}.Result) {name=Corners Roundness, nodeColor=<rgba 0 0 0 1>}
  NamedRerouteDeclaration(None=MaterialFunctionCall!(UVs=NamedRerouteUsage(), Slant amount=NamedRerouteUsage()) {materialFunction=/FNE_UI/Systems/_MaterialFunctions/Effects/MF_FNEUI_Slant}.Slanted UVs) {name=Shape Slanted UV, nodeColor=<rgba 1 0.0036 0 1>}
  n8 = ScalarParameter() {
   desc="Default Value = 0.0\r\n\r\nDefine shape X-Axis inclination. Value can be positive or negative.\r\n"
   group="[1] Fill"
   parameterName=SlantAmountX
   sortPriority=23
  }
  n9 = ScalarParameter() {
   desc="Default Value = 0.0\r\n\r\nDefine shape Y-Axis inclination. Value can be positive or negative.\r\n"
   group="[1] Fill"
   parameterName=SlantAmountY
   sortPriority=23
  }
  NamedRerouteDeclaration(None=AppendVector(A=@n8, B=@n9)) {name=Slant Amount, nodeColor=<rgba 0 0 0 1>}
  n10 = ScalarParameter() {
   desc="0.0 = Use 9-slice UV (default)\r\nGreater than 0.0 = Use Normalized UV\r\n\r\nTips:\r\n- 9-sliced UV: if you apply 9-slice to your widget, your material will follow that cut.\r\n- Normalized UV: use if you don't want your material to be affected by 9-slice.\r\n"
   group="[1] Fill"
   parameterName=UVType
   sliderMax=3
   sliderMin=1
   sortPriority=1
  }
  NamedRerouteDeclaration(None=Saturate(None=Ceil(None=@n10))) {name=Common UV Type, nodeColor=<rgba 0 0 0 1>}
  n11 = NamedRerouteUsage()
  n12 = NamedRerouteUsage()
  n13 = NamedRerouteUsage()
  n14 = NamedRerouteUsage()
  n15 = Constant()
  n16 = NamedRerouteUsage()
  n17 = NamedRerouteUsage()
  NamedRerouteDeclaration(None=Saturate(None=MaterialFunctionCall!(UVs=@n11, Size=@n12, Outline Thickness=@n13, Glow Min=@n14, Glow Max=@n15, Corner Roundness=@n16, Keep Aspect Ratio=@n17) {materialFunction=/FNE_UI/Systems/_MaterialFunctions/Shapes/MF_FNEUI_Box}.Outline)) {name=Shape Stroke, nodeColor=<rgba 0 0 0 1>}
  n18 = NamedRerouteUsage()
  n19 = NamedRerouteUsage()
  n20 = NamedRerouteUsage()
  NamedRerouteDeclaration(None=LinearInterpolate(A=@n18, B=LinearInterpolate(A=@n18, B=@n19, Alpha=@n20), Alpha=NamedRerouteUsage())) {name=Idle Color or Gradient, nodeColor=<rgba 0 0 0 1>}
  n21 = VectorParameter() {
   desc="Sets the primary color shown behind transparent pixels. \r\n\r\nWhen you reduce the transparency of an object, it fades over this color, which affects how the object appears when partially transparent.\r\n"
   group="[3] Colors"
   parameterName=TransparencyColor1
   sortPriority=5
   defaultValue=<rgba 0 0 0 1>
  }
  NamedRerouteDeclaration(None=Saturate(None=MaterialFunctionCall(X=@n21.R, Y=@n21.R, Z=@n21.R, A=@n21.R) {materialFunction=&MakeFloat4}.Result)) {name=Color for Transparency 1, nodeColor=<rgba 0 0 0 1>}
  n22 = VectorParameter() {
   desc="Sets the secondary color shown behind transparent pixels when gradient is enabled.\r\n"
   group="[3] Colors"
   parameterName=TransparencyColor2
   sortPriority=6
   defaultValue=<rgba 0 0 0 1>
  }
  NamedRerouteDeclaration(None=Saturate(None=MaterialFunctionCall(X=@n22.R, Y=@n22.R, Z=@n22.R, A=@n22.R) {materialFunction=&MakeFloat4}.Result)) {name="Color for Transparency 2 (Gradient on)", nodeColor=<rgba 0 0 0 1>}
  n23 = NamedRerouteUsage()
  n24 = NamedRerouteUsage()
  n25 = NamedRerouteUsage()
  NamedRerouteDeclaration(None=LinearInterpolate(A=@n23, B=LinearInterpolate(A=@n23, B=@n24, Alpha=@n25), Alpha=NamedRerouteUsage())) {name=Color for Transparency - Color or Gradient, nodeColor=<rgba 0 0 0 1>}
  n26 = NamedRerouteUsage()
  n27 = Distance(A=ConstantBiasScale(None=@n26) {bias=-0.5, scale=2}, B=Constant())
  n28 = NamedRerouteUsage()
  n29 = NamedRerouteUsage()
  n30 = Subtract(A=LinearInterpolate(A=ComponentMask(None=@n26) {r=true}, B=Subtract(A=@n27, B=@n28), Alpha=NamedRerouteUsage()), B=@n29)
  n31 = Subtract(A=NamedRerouteUsage(), B=@n29)
  n32 = If(A=@n31, "A > B"=@n31, "A == B"=Constant() {r=0}, "A < B"=@n31)
  NamedRerouteDeclaration(None=Saturate(None=Divide(A=@n30, B=@n32))) {name=gradient, nodeColor=<rgba 0 0 0 1>}
  n33 = MaterialFunctionCall() {materialFunction=&GetUserInterfaceUV}
  NamedRerouteDeclaration(None=LinearInterpolate(A=@n33.9-Slice UV, B=@n33.Normalized UV, Alpha=NamedRerouteUsage())) {name=Common UV, nodeColor=<rgba 1 0.0036 0 1>}
  n34 = ScalarParameter() {
   defaultValue=1
   desc="Default Value = 1.0\r\nRange: 0.0 - 1.0\r\n\r\nSet the transparency between 0 (transparent) and 1 (opaque).\r\n"
   group="[1] Fill"
   parameterName=FillTransparency
   sliderMax=1
   sortPriority=2
  }
  NamedRerouteDeclaration(None=@n34) {name=Idle Fill Transparency, nodeColor=<rgba 0 0 0 1>}
  n35 = ScalarParameter() {
   desc="Default value = 0.0\r\nRange: -1.0 - 1.0\r\n\r\nRotate your shape.\r\n\r\nWarning:\r\nRotating an object can distort its appearance unless you:\r\n- Change the UV Scaling Type to Scale to Fit or Scale to Fill.\r\n- Use a square widget size.\r\n"
   group="[1] Fill"
   parameterName=RotationAngle
   sliderMax=1
   sliderMin=-1
   sortPriority=17
  }
  NamedRerouteDeclaration(None=@n35) {name=Shape Rotation Angle, nodeColor=<rgba 0 0 0 1>}
  n36 = ScalarParameter() {
   desc="0.0 = No (default)\r\nGreater than 0.0 = Yes\r\n\r\nTips:\r\n- Disabled: The stroke is drawn on top of the fill.\r\n- Enabled: The stroke is cut out from the fill, leaving no fill behind it. Reducing stroke opacity will reveal the space where the stroke overlaps.\r\n"
   group="[2] Stroke"
   parameterName=SubtractStrokeFromFill
   sliderMax=1
   sortPriority=7
  }
  NamedRerouteDeclaration(None=Saturate(None=Ceil(None=@n36))) {name=Subtract Stroke From Fill, nodeColor=<rgba 0 0 0 1>}
  n37 = VectorParameter() {
   desc="Set the primary stroke color.\r\n\r\nThe alpha channel controls only the color’s opacity, which can be useful when working with gradients.\r\n"
   group="[3] Colors"
   parameterName=StrokeColor1
   sortPriority=3
   defaultValue=<rgba 0 0.333 1 1>
  }
  NamedRerouteDeclaration(None=Saturate(None=MaterialFunctionCall(X=@n37.R, Y=@n37.R, Z=@n37.R, A=@n37.R) {materialFunction=&MakeFloat4}.Result)) {name=Idle Stroke Color 1, nodeColor=<rgba 0 0 0 1>}
  n38 = VectorParameter() {
   desc="Set the secondary stroke color when gradient is enabled.\r\n\r\nThe alpha channel controls only the color’s opacity, which can be useful when working with gradients.\r\n"
   group="[3] Colors"
   parameterName=StrokeColor2
   sortPriority=4
   defaultValue=<rgba 1 0.333 0 1>
  }
  NamedRerouteDeclaration(None=Saturate(None=MaterialFunctionCall(X=@n38.R, Y=@n38.R, Z=@n38.R, A=@n38.R) {materialFunction=&MakeFloat4}.Result)) {name="Idle Stroke Color 2 (gradient on)", nodeColor=<rgba 0 0 0 1>}
  n39 = NamedRerouteUsage()
  n40 = NamedRerouteUsage()
  n41 = NamedRerouteUsage()
  NamedRerouteDeclaration(None=LinearInterpolate(A=@n39, B=LinearInterpolate(A=@n39, B=@n40, Alpha=@n41), Alpha=NamedRerouteUsage())) {name=Idle Stroke Color or Gradient, nodeColor=<rgba 0 0 0 1>}
  n42 = MaterialFunctionCall(
   UVs=NamedRerouteUsage()
   Size=NamedRerouteUsage()
   Outline Thickness=NamedRerouteUsage()
   Glow Min=NamedRerouteUsage()
   Glow Max=Constant()
   Corner Roundness=NamedRerouteUsage()
   Keep Aspect Ratio=NamedRerouteUsage()
  ) {materialFunction=&MF_FNEUI_Box}
  n43 = NamedRerouteUsage()
  NamedRerouteDeclaration(None=Saturate(None=Subtract(A=@n42.Outline, B=@n43))) {name=Gap, nodeColor=<rgba 0 0 0 1>}
  n44 = MaterialFunctionCall() {materialFunction=&GetUserInterfaceUV}
  n45 = ComponentMask(None=@n44.Pixel Size) {r=true}
  n46 = ComponentMask(None=@n44.Pixel Size) {g=true}
  NamedRerouteDeclaration(None=Min(
    A=NamedRerouteDeclaration(None=@n45) {name=pixelSize.x, nodeColor=<rgba 0 0.8125 1 1>}
    B=NamedRerouteDeclaration(None=@n46) {name=pixelSize.y, nodeColor=<rgba 1 0.5391 0 1>}
   )) {name=Pixel Size Min, nodeColor=<rgba 0.1484 0 1 1>}
  n47 = MaterialFunctionCall() {materialFunction=&ScreenResolution}
  n48 = ComponentMask(None=@n47.Visible Resolution) {r=true}
  n49 = ComponentMask(None=@n47.Visible Resolution) {g=true}
  NamedRerouteDeclaration(None=Min(
    A=NamedRerouteDeclaration(None=@n48) {name=ScreenResolution.x, nodeColor=<rgba 0 0.8125 1 1>}
    B=NamedRerouteDeclaration(None=@n49) {name=ScreenResolution.y, nodeColor=<rgba 1 0.5391 0 1>}
   )) {name=Screen Resolution Min, nodeColor=<rgba 0.0309 0.0675 1 1>}
  n50 = ScalarParameter() {
   defaultValue=0.98
   desc="Default Value = 0.98 \r\nRange: 0.0 - 1.0\r\n\r\nChange how sharp the edges of the shape are.\r\n"
   group="[1] Fill"
   parameterName=Sharpness
   sliderMax=1
   sortPriority=7
  }
  MaterialExpressionRemap_0 = MaterialXRemap!(None=@n50, InputLow, InputHigh, TargetLow, TargetHigh) {targetHighDefault=0, targetLowDefault=-0.1}
  NamedRerouteDeclaration(None=Clamp(None=@MaterialExpressionRemap_0) {desc=0 can not work..., maxDefault=-0.0001, minDefault=-0.1}) {name=Shape Sharpness, nodeColor=<rgba 0 0 0 1>}
  NamedRerouteDeclaration(None=MaterialFunctionCall(UVs=NamedRerouteUsage(), "Rotation Angle (0-1)"=NamedRerouteUsage()) {materialFunction=&CustomRotator}.Rotated Values) {name=Shape Rotated UV, nodeColor=<rgba 1 0.0036 0 1>}
  n51 = ScalarParameter() {
   defaultValue=1
   desc="Default Value = 1.0\r\n\r\nRelative to the scale parameter. \r\n"
   group="[2] Stroke"
   parameterName=StrokeScale
   sortPriority=3
  }
  NamedRerouteDeclaration(None=@n51) {name=Stroke Scale, nodeColor=<rgba 0 0 0 1>}
  NamedRerouteDeclaration() {name=Keep Consistant Stroke Thickness, nodeColor=<rgba 0 0 0 1>}
  n52 = Ceil(None=NamedRerouteUsage())
  NamedRerouteDeclaration(None=LinearInterpolate(A=NamedRerouteUsage(), B=NamedRerouteUsage(), Alpha=Saturate(None=@n52))) {name=Final stroke Thickness, nodeColor=<rgba 0 0 0 1>}
  n53 = Multiply(A=NamedRerouteUsage()) {constB=0.1}
  n54 = NamedRerouteUsage()
  NamedRerouteDeclaration(None=Divide(A=Multiply(A=@n53, B=@n54), B=NamedRerouteUsage())) {name=Consistant Stroke Thickness, nodeColor=<rgba 0 0 0 1>}
  n55 = NamedRerouteUsage()
  n56 = NamedRerouteUsage()
  n57 = ScalarParameter() {
   defaultValue=1
   desc="Default Value = 1.0\r\n\r\nCorrects uneven stroke corners caused by scaling.\r\n\r\nChanging stroke width or height can cause the stroke corners to appear different from the fill, which may look uneven. This option allows you to fine-tune the roundness of stroke corners to match your desired stroke scale and size.\r\n"
   group="[2] Stroke"
   parameterName=StrokeCornerRoundnessScale
   sortPriority=6
  }
  NamedRerouteDeclaration(None=Multiply(A=Multiply(A=@n55, B=@n56) {desc="Make the stroke Corner roundness relative \r\nto the Corners Roundness Parameter."}, B=@n57)) {name=Stroke Corner Roundness, nodeColor=<rgba 0 0 0 1>}
  n58 = ScalarParameter() {
   defaultValue=1
   desc="Default Value = 1.0\r\n\r\nRelative to the width parameter.\r\n"
   group="[2] Stroke"
   parameterName=StrokeWidth
   sortPriority=4
  }
  n59 = NamedRerouteUsage()
  n60 = Multiply(A=@n58, B=@n59)
  n61 = ScalarParameter() {
   defaultValue=1
   desc="Default Value = 1.0\r\n\r\nRelative to the height parameter. \r\n"
   group="[2] Stroke"
   parameterName=StrokeHeight
   sortPriority=5
  }
  n62 = Multiply(A=@n59, B=@n61)
  NamedRerouteDeclaration(None=Multiply(A=AppendVector(A=@n60, B=@n62), B=NamedRerouteUsage())) {name=Stroke Size, nodeColor=<rgba 0 0 0 1>}
  n63 = ScalarParameter() {
   defaultValue=1
   desc="Default Value = 1.0\r\n\r\nScale your shape from 0 to 1 (100%) \r\n"
   group="[1] Fill"
   parameterName=Scale
   sliderMax=1
   sortPriority=4
  }
  NamedRerouteDeclaration(None=@n63) {name=Scale, nodeColor=<rgba 0 0 0 1>}
  n64 = ScalarParameter() {
   defaultValue=1
   desc="1.0 = Keep the corner aspect ratio (default)\r\nLower than 1.0 = Let the box stretch\r\n\r\nWarning:\r\nIf your ScalingMethod is not set to stretch, you have to set ‘KeepCornersAspectRatio’ value to 0 to avoid cancelling the expected result.\r\n"
   group="[1] Fill"
   parameterName=KeepCornersAspectRatio
   sliderMax=1
   sortPriority=12
  }
  NamedRerouteDeclaration(None=@n64) {name=Corner Aspect Ratio, nodeColor=<rgba 0 0 0 1>}
  n65 = NamedRerouteUsage()
  n66 = NamedRerouteUsage()
  NamedRerouteDeclaration(None=MaterialFunctionCall!(
    "Value to Ease (0-1)"=MaterialFunctionCall!(No Time Static Value, Speed=@n65, Time Style Switch Index=@n66) {materialFunction=/FNE_UI/Systems/_MaterialFunctions/Time/MF_FNEUI_TimeSwitcher}.Selected Time Style
    Easing Curve Switch Index=NamedRerouteUsage()
   ) {materialFunction=/FNE_UI/Systems/_MaterialFunctions/Animation/MF_FNEUI_EaseCurveSwitcher}.Selected Easing Curve) {name=Gradient Animated Rotation, nodeColor=<rgba 0 0 0 1>}
  n67 = ScalarParameter() {
   desc="0.0 = Gradient off (default)\r\nGreater than 0.0 = Gradient on\r\n\r\nEnabled: uses Color1 and Color2 to create a gradient.\r\nDisabled: uses only Color1.\r\n"
   group="[3] Colors"
   parameterName=GradientOn
   sliderMax=1
   sortPriority=7
  }
  NamedRerouteDeclaration(None=Saturate(None=Ceil(None=@n67))) {name=Gradient On / Off, nodeColor=<rgba 0 0 0 1>}
  n68 = VectorParameter() {
   desc="Sets the primary color.\r\n\r\nThe alpha channel controls only the color’s opacity, which can be useful when working with gradients.\r\n"
   group="[3] Colors"
   parameterName=Color1
   sortPriority=1
   defaultValue=<rgba 1 1 1 1>
  }
  NamedRerouteDeclaration(None=Saturate(None=MaterialFunctionCall(X=@n68.R, Y=@n68.R, Z=@n68.R, A=@n68.R) {desc=Needed to get the Alpha of the color, materialFunction=&MakeFloat4}.Result)) {name=Idle Color 1, nodeColor=<rgba 0 0 0 1>}
  n69 = VectorParameter() {
   desc="Sets the secondary color when gradient is enabled.\r\n\r\nThe alpha channel controls only the color’s opacity, which can be useful when working with gradients.\r\n"
   group="[3] Colors"
   parameterName=Color2
   sortPriority=2
   defaultValue=<rgba 0 0.33 1 1>
  }
  NamedRerouteDeclaration(None=Saturate(None=MaterialFunctionCall(X=@n69.R, Y=@n69.R, Z=@n69.R, A=@n69.R) {materialFunction=&MakeFloat4}.Result)) {name="Idle Color 2 (gradient on)", nodeColor=<rgba 0 0 0 1>}
  n70 = ScalarParameter() {
   desc="0.0 = Linear Gradient (default)\r\n1.0 = Radial Gradient\r\n"
   group="[3] Colors"
   parameterName=RadialGradient
   sliderMax=1
   sortPriority=8
  }
  NamedRerouteDeclaration(None=Saturate(None=Ceil(None=@n70))) {name=Linear or Radial Gradient, nodeColor=<rgba 0 0 0 1>}
  n71 = ScalarParameter() {desc="Default Value = 0.0\r\n\r\nStarting point of the gradient.\r\n", group="[3] Colors", parameterName=GradientStart, sortPriority=10}
  NamedRerouteDeclaration(None=@n71) {name=Gradient Start, nodeColor=<rgba 0 0 0 1>}
  n72 = ScalarParameter() {
   defaultValue=0.015
   desc="Default Value = 0.015\r\n\r\nChange the thickness of the stroke.\r\n"
   group="[2] Stroke"
   parameterName=StrokeThickness
   sliderMax=1
   sortPriority=2
  }
  n73 = ScalarParameter() {
   desc="Default value = 0.0\r\n\r\nCreates a space between the stroke and the fill. The gap’s appearance depends on stroke thickness, size, and scale.\r\n\r\nTips:\r\n- The gap remains even if stroke transparency is 0, since it actually cuts into the fill.\r\n- Setting SubtractStrokeFromFill = 1 hides the fill where the stroke would appear (even if the stroke is fully transparent).\r\n"
   group="[2] Stroke"
   parameterName=GapThickness
   sliderMax=1
   sortPriority=3
  }
  NamedRerouteDeclaration(None=Add(A=@n72, B=@n73)) {name=GapThickness, nodeColor=<rgba 0 0 0 1>}
  n74 = ScalarParameter() {
   defaultValue=1
   desc="Default Value = 1.0\r\n\r\nEnding point of the gradient.\r\n"
   group="[3] Colors"
   parameterName=GradientEnd
   sortPriority=11
  }
  NamedRerouteDeclaration(None=@n74) {name=Gradient End, nodeColor=<rgba 0 0 0 1>}
  n75 = NamedRerouteUsage()
  n76 = NamedRerouteUsage()
  n77 = NamedRerouteUsage()
  NamedRerouteDeclaration(None=LinearInterpolate(A=LinearInterpolate(A=@n75, B=@n76, Alpha=@n77), B=NamedRerouteUsage(), Alpha=NamedRerouteUsage())) {name=Idle Final Color, nodeColor=<rgba 0 0 0 1>}
  n78 = ScalarParameter() {
   defaultValue=1
   desc="Default Value = 1.0\r\n\r\nFor radial gradients, adjust the size of the gradient circle.\r\n"
   group="[3] Colors"
   parameterName=RadialGradientSizeFactor
   sortPriority=12
  }
  NamedRerouteDeclaration(None=Subtract(A=@n78) {desc=Make more sense for the user}) {name=Radial Gradient Scale, nodeColor=<rgba 0 0 0 1>}
  NamedRerouteDeclaration(None=NamedRerouteUsage()) {name=Shape UV, nodeColor=<rgba 1 0.0036 0 1>}
  n79 = ScalarParameter() {
   desc="Default Value = 0.0\r\n\r\nShifts the gradient away from its center in any direction (0 = center).\r\n\r\nFor radial gradients, set an offset to see the effect of gradient rotation. Without it, the gradient rotates around its center and appears static.\r\n"
   group="[3] Colors"
   parameterName=GradientOffset
   sortPriority=13
  }
  NamedRerouteDeclaration(None=@n79) {name=Gradient Offset, nodeColor=<rgba 0 0 0 1>}
  n80 = NamedRerouteUsage()
  n81 = NamedRerouteUsage() {desc="If 0, the UV scaling is comming \r\nfrom the shape. This one is ignored."}
  n82 = Saturate(None=Floor(None=@n81))
  NamedRerouteDeclaration(None=MaterialFunctionCall!(UV=NamedRerouteUsage(), Texture Size, UV Scaling Switch Index=LinearInterpolate(A=@n80, B=@n81, Alpha=@n82)) {materialFunction=/FNE_UI/Systems/_MaterialFunctions/Basics/MF_FNEUI_UVScalingSwitcher}.Selected Scaled UV) {name=Gradient Scaled UV, nodeColor=<rgba 1 0.0036 0 1>}
  n83 = ScalarParameter() {
   desc="0.0 = Use the UV ScalingMethod from the shape (default)\r\n1.0 = Stretch\r\n2.0 = Scaled to Fit\r\n3.0 = Scaled to Fill\r\n\r\nMethods 2 and 3 keep the aspect ratio of the gradient. \r\n\r\nWarning:\r\nFor stretched widgets, apply Scale to Fit or Scale to Fill when rotating to prevent distortion.\r\n"
   group="[3] Colors"
   parameterName=GradientScalingMethod
   sliderMax=3
   sortPriority=9
  }
  NamedRerouteDeclaration(None=@n83) {name=Gradient UV Scaling, nodeColor=<rgba 0 0 0 1>}
  NamedRerouteDeclaration(None=NamedRerouteUsage()) {name=gradient UV, nodeColor=<rgba 1 0.0036 0 1>}
  n84 = NamedRerouteUsage()
  n85 = Add(A=NamedRerouteUsage(), B=NamedRerouteUsage())
  NamedRerouteDeclaration(None=Subtract(
    A=MaterialFunctionCall(UVs=@n84, "Rotation Angle (0-1)"=@n85) {materialFunction=&CustomRotator}.Rotated Values
    B=NamedRerouteUsage()
   )) {name=gradient Rotated UV, nodeColor=<rgba 1 0.0036 0 1>}
  n86 = ScalarParameter() {
   desc="Default Value = 0.0\r\nRange: -1.0 - 1.0\r\n\r\nRotate your gradient.\r\n\r\nTips:\r\n- When GradientRotationAnimationType is set to ‘loop’ or ‘ping-pong’, the rotation angle sets the starting point of the animation.\r\n- For radial gradients, set an offset to see the effect of gradient rotation. Without it, the gradient rotates around its center and appears static.\r\n"
   group="[3] Colors"
   parameterName=GradientRotationAngle
   sliderMax=1
   sliderMin=-1
   sortPriority=14
  }
  NamedRerouteDeclaration(None=@n86) {name=Gradient Rotation Angle, nodeColor=<rgba 0 0 0 1>}
  NamedRerouteDeclaration(None=@n72) {name=Stroke Thickness, nodeColor=<rgba 0 0 0 1>}
  n87 = ScalarParameter() {
   desc="0.0 = Static (default)\r\n1.0 = Loop\r\n2.0 = Ping-Pong\r\n3.0 = Switch (A than B)\r\n\r\nSets the animation style of infinite gradient rotation. Rotation speed must be greater than 0.\r\n\r\nWarning:\r\nTexture rotation is applied within its rendering container. On non-square rectangles, this may cause distortion unless you:\r\n- Match width and height for both the material size and the image/widget size (the material container). \r\n- Use a UV ScalingMethod such as Scale to Fit or Scale to Fill.\r\n"
   group="[3] Colors"
   parameterName=GradientRotationAnimationType
   sortPriority=15
  }
  NamedRerouteDeclaration(None=@n87) {name="Gradient  Rotation Animation Type", nodeColor=<rgba 0 0 0 1>}
  n88 = ScalarParameter() {
   defaultValue=0.15
   desc="Default Value = 0.15\r\n\r\nA positive value rotates clockwise, while a negative value rotates counterclockwise. The see automatic rotation, a GradientRotationAnimationType must be set.\r\n"
   group="[3] Colors"
   parameterName=GradientRotationSpeed
   sortPriority=17
  }
  NamedRerouteDeclaration(None=@n88) {name=Gradient Rotation Speed, nodeColor=<rgba 0 0 0 1>}
  n89 = MaterialFunctionCall(
   UVs=NamedRerouteUsage()
   Size=NamedRerouteUsage()
   Outline Thickness=NamedRerouteUsage()
   Glow Min=NamedRerouteUsage()
   Glow Max=Constant()
   Corner Roundness=NamedRerouteUsage()
   Keep Aspect Ratio=NamedRerouteUsage()
  ) {materialFunction=&MF_FNEUI_Box}
  n90 = LinearInterpolate(A=@n89.Fill, B=Subtract(A=@n89.Fill, B=NamedRerouteUsage()), Alpha=NamedRerouteUsage())
  n91 = NamedRerouteUsage()
  NamedRerouteDeclaration(None=Saturate(None=Subtract(A=@n90, B=@n91))) {name=Shape Fill, nodeColor=<rgba 0 0 0 1>}
  n92 = ScalarParameter() {
   desc="0.0 = No Ease used (Default) \r\n1.0 =EaseInSine \r\n2.0 = EaseOutSine \r\n3.0 = EaseInOutSine \r\n4.0 = EaseInCubic \r\n5.0 = EaseOutCubic \r\n6.0 = EaseInOutCubic \r\n7.0 = EaseInQuint \r\n8.0 = EaseOutQuint \r\n9.0 = EaseInOutQuint \r\n\r\nThe easing curve of the animation. \r\n"
   group="[3] Colors"
   parameterName=GradientRotationEasingStyle
   sortPriority=16
  }
  NamedRerouteDeclaration(None=@n92) {name="Gradient  Rotation Easing Style", nodeColor=<rgba 0 0 0 1>}
  n93 = ScalarParameter() {
   defaultValue=1
   desc="1.0 = Stretch (default)\r\n2.0 = Scaled to Fit\r\n3.0 = Scaled to Full\r\n\r\nTips:\r\nIf your image or widget size is not a square, consider adjusting this parameter to maintain your shape's aspect ratio. Options 2 and 3 keep the aspect ratio of the shape.\r\n\r\nWarning:\r\n- Rotation on a stretched widget requires a scale to fit or fill, avoiding distortions.\r\n- When active, you must disable the ‘KeepCornersAspectRatio’ parameter to avoid cancelling the expected result.\r\n"
   group="[1] Fill"
   parameterName=ScalingMethod
   sliderMax=3
   sliderMin=1
   sortPriority=3
  }
  NamedRerouteDeclaration(None=Floor(None=@n93)) {name=Shape UV Scaling, nodeColor=<rgba 0 0 0 1>}
  NamedRerouteDeclaration(None=MaterialFunctionCall(UV=NamedRerouteUsage(), UV Scaling Switch Index=NamedRerouteUsage()) {materialFunction=&MF_FNEUI_UVScalingSwitcher}.Selected Scaled UV) {name=Shape Scaled UV, nodeColor=<rgba 1 0.0036 0 1>}
  n94 = NamedRerouteUsage()
  n95 = NamedRerouteUsage()
  n96 = NamedRerouteUsage()
  n97 = Saturate(None=Multiply(A=Multiply(A=@n94, B=@n95), B=ComponentMask(None=@n96) {a=true, desc="The alpha of the Shape Color \r\nalso drives the shape alpha"}))
  n98 = NamedRerouteUsage()
  n99 = NamedRerouteUsage()
  n100 = NamedRerouteUsage()
  n101 = Saturate(None=Multiply(A=Multiply(A=@n98, B=@n99), B=ComponentMask(None=@n100) {a=true, desc="The alpha of the stroke color \r\nalso drives the stroke alpha"}))
  NamedRerouteDeclaration(None=Max(
    A=NamedRerouteDeclaration(None=@n97) {name=Idle Shape Alpha, nodeColor=<rgba 0 0 0 1>}
    B=NamedRerouteDeclaration(None=@n101) {name=Idle Stroke Alpha, nodeColor=<rgba 0 0 0 1>}
   )) {name=Idle Alpha, nodeColor=<rgba 0 0 0 1>}
  n102 = ScalarParameter() {
   desc="Default value = 0.0\r\nRange: 0.0 - 1.0\r\n\r\nSet the transparency between 0 (transparent) and 1 (opaque).\r\n\r\nThe fill may hide the stroke. Reduce FillTransparency to make it visible.\r\n"
   group="[2] Stroke"
   parameterName=StrokeTransparency
   sliderMax=1
   sortPriority=1
  }
  NamedRerouteDeclaration(None=@n102) {name=Idle Stroke Transparency, nodeColor=<rgba 0 0 0 1>}
  NamedRerouteUsage()
  NamedRerouteUsage()
  ScalarParameter() {
   desc="\"Reference Size\" but named \"Consistant Stroke Thikness\" for players 'cause it is replacing the stroke thickness.\r\nif 0, not activated. \r\nfor any other value,  stroke thickness = 1 is 1 for this reference size."
   group="[2] Stroke"
   parameterName=StrokeConsistantThiknessOn
   sliderMax=1
   sortPriority=2
  }
 ]
}
